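# Niagara skydiving contrail that combines a rotating mesh swirl, looping torus splash sprites and orbiting popcorn that pops.
# Popcorn (GPU) and PopcornLow (CPU) switch Spawngroup at a random age, swapping mesh by index or visibility tag.
# An integrator then times post-pop scale, sprite smoke and kill.
# Popped pieces accelerate using the diving module's Dive value, with a separate front-end vector; MID_splash is disabled.
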NiagaraSystem /BRCosmetics/Athena/Effects/Particles/Skydiving/NS_Contrail_Popcorn/NS_Contrail_PopcornSwirl_Converted
dependencies {
 usedDataInterfaces [
  &NiagaraDataInterfaceArrayFloat
  /Script/Niagara.NiagaraDataInterfaceArrayFloat4
  &NiagaraDataInterfaceVectorCurve
  &NiagaraDataInterfaceCurve
 ]
 usedDynamicInputs [&FN_FrontEndFloatModifier]
 usedModules [
  &SystemState
  /Game/Effects/Niagara/Modules/Skydiving/FN_DivingAttributes/v5/FN_DivingAttributes
  &self:SystemSpawnScript.SystemScriptSource.SystemScriptGraph.NiagaraNodeAssignment_0.SetVariables_#1
 ]
}
emitters {
 InitalSplash {
  emitterUpdateScript [
   &EmitterState(Life Cycle Mode=::NewEnumerator1, Loop Delay=1.2, UseLoopDelay=false)
   &SpawnRate(
    Spawn Probability=&FloatFromCurve(
     FloatCurve=DataInterface() {propertyValues {Curve {keys [
         {arriveTangent=-21.1376, interpMode=RCIM_Constant, time=0, value=0}
         <at 0.019 4 RCIM_Cubic>
         {arriveTangent=-4.0775, interpMode=RCIM_Cubic, leaveTangent=-4.0775, time=0.06, value=0}
         <at 1 0>
        ]}}}
     CurveIndex=$Emitter.NormalizedLoopAge
     Scale Curve=&MakeCustomFloatFromBool(
      Bool=&SetBoolByIntComparison(A=$Emitter.LoopCount, B=1, Comparison Type=::NewEnumerator2)
      True Float=0
      False Float=1
     )
    )
    SpawnRate=50
    Use Spawn Probability=true
   ) {module=SpawnRate001}
  ]
  particleSpawnScript [
   /Niagara/Modules/Spawn/Initialization/InitializeParticle(
    Lifetime=&RandomRangeFloat(Minimum=0.9, Maximum=1.7, Evaluation Type=::NewEnumerator1)
    Module.Write Color=true
    Module.Write Lifetime=true
    Module.Write Mass=false
    Module.Write Material Random=true
    Module.Write Position=true
    Module.Write Scale=false
    Module.Write SpriteRotation=true
    Module.Write SpriteSize=true
    Sprite Rotation=&NormalizedAngleToDegrees(NormalizedAngle=&RandomRangeFloat(Minimum=0, Maximum=1, Evaluation Type=::NewEnumerator1))
    Sprite Size=&Vector2DFromFloat(Value=&RandomRangeFloat(Minimum=70, Maximum=90, Evaluation Type=::NewEnumerator1))
   )
   &SystemLocation(MaskSpawnGroup=false, Offset=<0 0 -25>)
   &Color(Color=&MakeLinearColorFromVectorAndFloat("Vector (RGB)"=<1 1 1>, "Float (Alpha)"=&ClampFloat(Float=0.8)))
   &DynamicMaterialParameters(Module.Param0WriteEnabled=true, Module.Param1WriteEnabled=false, Module.Param2WriteEnabled=false, Module.Param3WriteEnabled=false)
   &AddVelocity(Velocity=&RandomRangeVector(Minimum=<0 0 300>, Maximum=<0 0 400>, Evaluation Type=::NewEnumerator1))
   &TorusLocation(Align Torus With Axis=false, Large Radius=105.8244, MaskSpawnGroup=false, Module.Override Randomness=false)
  ]
  particleUpdateScript [
   &ParticleState()
   &ScaleSpriteSize(Scale Factor=&Vector2DFromFloat(Value=&FloatFromCurve(FloatCurve=DataInterface() {propertyValues {Curve {keys [<at -0.0011 0.0573 RCIM_Cubic>, {arriveTangent=1.9407, interpMode=RCIM_Cubic, leaveTangent=1.9407, time=0.4308, value=1}, <at 1 2>]}}})))
   &ScaleColor(
    Scale Alpha=&FloatFromCurve(
     FloatCurve=DataInterface() {propertyValues {Curve {keys [
         <at 0 0 RCIM_Cubic>
         {arriveTangent=-0.0427, interpMode=RCIM_Cubic, leaveTangent=-0.0427, tangentMode=RCTM_User, time=0.2, value=1}
         <at 1 0>
        ]}}}
     Scale Curve=&RandomRangeFloat(Minimum=0.8, Maximum=1)
    )
    Scale RGB=&VectorFromCurve(VectorCurve=DataInterface() {propertyValues {XCurve {keys [<at 0 1>,<at 1 1>]}, YCurve {keys [<at 0 1>,<at 1 1>]}, ZCurve {keys [<at 0 1>,<at 1 1>]}}})
    ScaleA=true
    ScaleRGB=true
   )
   &DynamicMaterialParameters(
    Index 0 Param 4=1
    Module.Param0WriteEnabled=false
    Module.Param1WriteEnabled=true
    Module.Param2WriteEnabled=true
    Module.Param3WriteEnabled=true
   ) {module=DynamicMaterialParameters001}
   /Niagara/Modules/Update/SubUV/SubUVAnimation()
   &AccelerationForce(Acceleration=<0 0 -900>)
   &SolveForcesAndVelocity(Clamp Velocity=false, Limit Acceleration=false)
  ]
  rendererClasses [&NiagaraSpriteRendererProperties]
  renderers [{rendererClass=&NiagaraSpriteRendererProperties}]
 }
 MID_splash {
  bEnabled=false
  emitterUpdateScript [
   &EmitterState()
   &SpawnRate(SpawnRate=10, Use Spawn Probability=false)
  ]
  particleSpawnScript [
   /SetVariables(Particles.Acceleration:Vector3f=<0 0 -900>) {isSetParameters=true}
   &InitializeParticle(
    Lifetime=&RandomRangeFloat(Minimum=0.6, Maximum=0.9, Evaluation Type=::NewEnumerator1)
    Module.Write Color=true
    Module.Write Lifetime=true
    Module.Write Mass=true
    Module.Write Material Random=true
    Module.Write Position=false
    Module.Write Scale=false
    Module.Write SpriteRotation=true
    Module.Write SpriteSize=true
    Sprite Rotation=&NormalizedAngleToDegrees(NormalizedAngle=&RandomRangeFloat(Minimum=0, Maximum=1, Evaluation Type=::NewEnumerator1))
    Sprite Size=&Vector2DFromFloat(Value=&RandomRangeFloat(Minimum=100, Maximum=120, Evaluation Type=::NewEnumerator1))
   )
   &Color(Color=&MakeLinearColorFromVectorAndFloat("Vector (RGB)"=<1 1 1>, "Float (Alpha)"=&ClampFloat(Float=0.89)))
   &DynamicMaterialParameters(Module.Param0WriteEnabled=true, Module.Param1WriteEnabled=false, Module.Param2WriteEnabled=false, Module.Param3WriteEnabled=false)
   &AddVelocity(Velocity=&RandomRangeVector(Minimum=<0 0 300>, Maximum=<0 0 400>, Evaluation Type=::NewEnumerator1))
   /Niagara/Modules/Spawn/Location/SphereLocation(
    MaskSpawnGroup=false
    Module.Override Randomness=false
    Non Uniform Scale=<1 1 0>
    Sphere Radius=&RandomRangeFloat(Minimum=80, Maximum=110, Evaluation Type=::NewEnumerator1)
    Surface Only=true
   )
  ]
  particleUpdateScript [
   &ParticleState()
   &ScaleSpriteSize(Scale Factor=&Vector2DFromFloat(Value=&FloatFromCurve(FloatCurve=DataInterface() {propertyValues {Curve {keys [<at -0.0011 0.0573 RCIM_Cubic>, {arriveTangent=1.3413, interpMode=RCIM_Cubic, leaveTangent=1.3413, time=0.4308, value=0.9897}, <at 1 1.4>]}}})))
   &ScaleColor(
    Scale Alpha=&FloatFromCurve(FloatCurve=DataInterface() {propertyValues {Curve {keys [<at 0 0 RCIM_Cubic>,<at 0.4 1 RCIM_Cubic>,<at 1 0>]}}})
    Scale RGB=&VectorFromCurve(VectorCurve=DataInterface() {propertyValues {XCurve {keys [<at 0 1>,<at 1 1>]}, YCurve {keys [<at 0 1>,<at 1 1>]}, ZCurve {keys [<at 0 1>,<at 1 1>]}}})
    ScaleA=true
    ScaleRGB=true
   )
   &DynamicMaterialParameters(
    Index 0 Param 4=1
    Module.Param0WriteEnabled=false
    Module.Param1WriteEnabled=true
    Module.Param2WriteEnabled=true
    Module.Param3WriteEnabled=true
   ) {module=DynamicMaterialParameters001}
   &SubUVAnimation()
   &AccelerationForce(Acceleration=$Particles.Acceleration)
   &SolveForcesAndVelocity(Clamp Velocity=false, Limit Acceleration=false)
  ]
  rendererClasses [&NiagaraSpriteRendererProperties]
  renderers [{rendererClass=&NiagaraSpriteRendererProperties}]
 }
 Popcorn {
  simTarget=GPUComputeSim
  emitterUpdateScript [
   &EmitterState()
   &SpawnRate(SpawnRate=6, Use Spawn Probability=false)
  ]
  particleSpawnScript [
   /SetVariables(Particles.MeshOrientation=&MakeQuatFromAxisAngle(
     Axis=&RandomRangeVector(Minimum=<-0.05 -0.1 0>, Maximum=<0.05 0.1 0>, Evaluation Type=::NewEnumerator1)
     AngleInDegrees=90
    )) {isSetParameters=true, module=SetVariables_#2}
   /SetVariables(Particles.SpawnRotationVector=&RandomRangeVector(Minimum=<0 0 0>, Maximum=<1 0.3 0>, Evaluation Type=::NewEnumerator1)) {isSetParameters=true, module=SetVariables_#3}
   /Niagara/Modules/Spawn/Initialization/V2/InitializeParticle(
    Lifetime Max=5
    Lifetime Min=4
    Lifetime Mode=::NewEnumerator1
    Mesh Scale Mode=::NewEnumerator1
    Position=&Add_Vector(
     A=/Niagara/DynamicInputs/Helpers/SimulationPosition()
     B=&MakeVector(X=0, Y=0, Z=&RandomRangeFloat(Minimum=0, Maximum=50))
    )
    Position Mode=::NewEnumerator1
    Sprite Size Mode=::NewEnumerator1
    Uniform Sprite Size Max=150
    Uniform Sprite Size Min=80
    UsePositionOffset=false
    Write Material Random=true
   ) {module=InitializeParticle001}
   &SphereLocation(
    MaskSpawnGroup=false
    Module.Override Randomness=false
    Non Uniform Scale=<1 1 0>
    Offset=<0 0 65>
    Sphere Radius=&RandomRangeFloat(Minimum=75, Maximum=140, Evaluation Type=::NewEnumerator1)
    Surface Only=true
   ) {enabled=false}
   /SetVariables(Particles.PopThreshold:NiagaraFloat=&RandomRangeFloat(Minimum=1, Maximum=2)) {isSetParameters=true, module=SetVariables_#4}
  ]
  particleUpdateScript [
   &ParticleState(Kill Particles When Lifetime Has Elapsed=false)
   /SetVariables(Particles.Spawngroup=HlslExpression() {hlslExpression="Particles.Age < Particles.PopThreshold ? 0 : 1"}) {isSetParameters=true, module=SetVariables_#5}
   &UpdateMeshOrientation(Rotation Vector=$Particles.SpawnRotationVector)
   /SetVariables(Particles.OrbitOffset:Vector3f=&RotateVectorByQuaternion(
     Vector To Rotate=&MakeVector(X=&RandomRangeFloat(Minimum=100, Maximum=150), Y=0, Z=0)
     Quaternion=&MakeQuatFromAxisAngle(
      Axis=<0 0 1>
      AngleInDegrees=&NormalizedAngleToDegrees(NormalizedAngle=&Add_Float(A=$Particles.NormalizedAge, B=&RandomRangeFloat(Minimum=0, Maximum=1)))
     )
    )) {isSetParameters=true, module=SetVariables_#6}
   /SetVariables(Particles.MeshIndex=$Particles.Spawngroup) {isSetParameters=true, module=SetVariables_#7}
   &FN_Integrator(Create Normalized Value=true, Increment=HlslExpression() {hlslExpression="Particles.Spawngroup == 1 ? 1 : 0"})
   &ScaleMeshSize(Scale Factor=&Lerp_Vector(
     A=&VectorFromFloat(Value=&FloatFromCurve(FloatCurve=DataInterface() {propertyValues {Curve {keys [<at 0 0 RCIM_Cubic>,<at 0.2 1.0034>]}}}))
     B=&VectorFromFloat(Value=&FloatFromCurve(
       FloatCurve=DataInterface() {propertyValues {Curve {keys [
           <at -0.0014 0.9856 RCIM_Cubic>
           {arriveTangent=0.0653, interpMode=RCIM_Cubic, leaveTangent=0.0653, time=0.0796, value=2}
           <at 0.22 1 RCIM_Constant>
           {arriveTangent=-1.281, interpMode=RCIM_Cubic, leaveTangent=-1.281, time=0.75, value=1}
           <at 1.0006 0>
          ]}}}
       CurveIndex=$Particles.FN_Integrator.NormalizedValue
      ))
     Alpha=&MakeFloatFromInt(Int=$Particles.Spawngroup)
    ))
   /SetVariables(Particles.RenderPosition:Vector3f=&Add_Vector(A=Linked() {linkedVariable {name=Particles.Position, type=NiagaraPosition}}, B=$Particles.OrbitOffset)) {isSetParameters=true, module=SetVariables_#8}
   &AccelerationForce(Acceleration=&FN_FrontEndVector3Modifier(
     Gameplay=&MakeVector(
      X=0
      Y=0
      Z=&Lerp_Float(
       A=0
       B=&Lerp_Float(
        A=HlslExpression() {hlslExpression=Particles.Spawngroup * 2400}
        B=HlslExpression() {hlslExpression=Particles.Spawngroup * 2900}
        Alpha=$System.Dive
       )
       Alpha=&Lerp_Float(
        A=&Lerp_Float(
         A=0
         B=1
         Alpha=&ClampFloat(Float=/Niagara/DynamicInputs/Multiply/ScaleAndBiasFloat(Float:NiagaraFloat=$Particles.NormalizedAge, Scale:NiagaraFloat=1.1, Bias:NiagaraFloat=-0.1))
        )
        B=2
        Alpha=$System.Dive
       )
      )
     )
     Front End=&Multiply_Vector(
      A=&TransformVector(Vector=&Lerp_Vector(A=<0 0 1000>, B=<1 -1000 1000>, Alpha=$System.Dive), Destination Space=::World)
      B=&VectorFromFloat(Value=HlslExpression() {hlslExpression=Particles.Spawngroup})
     )
    ))
   &SolveForcesAndVelocity(Clamp Velocity=false, Limit Acceleration=false)
   &DynamicMaterialParameters(
    Index 0 Param 1=&FloatFromCurve(
     FloatCurve=DataInterface() {propertyValues {Curve {keys [<at 0.5016 0 RCIM_Cubic>,<at 1 1>]}}}
     CurveIndex=$Particles.FN_Integrator.NormalizedValue
    )
    Index 0 Param 4=1
    Module.Param0WriteEnabled=true
    Module.Param1WriteEnabled=true
    Module.Param2WriteEnabled=true
    Module.Param3WriteEnabled=true
   ) {module=DynamicMaterialParameters001}
   &ScaleSpriteSize(Scale Factor=&Vector2DFromFloat(Value=&FloatFromCurve(
      FloatCurve=DataInterface() {propertyValues {Curve {keys [
          {arriveTangent=4.8897, interpMode=RCIM_Cubic, leaveTangent=4.8897, tangentMode=RCTM_User, time=-0.0011, value=0.0573}
          {arriveTangent=1.3413, interpMode=RCIM_Cubic, leaveTangent=1.3413, time=0.1954, value=1.173}
          <at 1 1.4>
         ]}}}
      CurveIndex=$Particles.FN_Integrator.NormalizedValue
     )))
   /SetVariables(Particles.SmokeColor:LinearColor=&MakeLinearColorFromVectorAndFloat(
     "Vector (RGB)"=<1 1 1>
     "Float (Alpha)"=&FloatFromCurve(
      FloatCurve=DataInterface() {propertyValues {Curve {keys [
          <at -0.0008 1>
          {arriveTangent=-3.6247, interpMode=RCIM_Cubic, leaveTangent=-3.6248, tangentMode=RCTM_User, time=0.053, value=1.3351}
          <at 0.5092 0 RCIM_Cubic>
         ]}}}
      CurveIndex=$Particles.FN_Integrator.Value
      Scale Curve=0.1
     )
    )) {isSetParameters=true, module=SetVariables_#9}
   /Niagara/Modules/Update/SubUV/V2/SubUVAnimation(End Frame=64)
   &KillParticles(Kill Particles=HlslExpression() {hlslExpression="Particles.FN_Integrator.NormalizedValue > 1 ? 1 :  0"})
  ]
  rendererClasses [&NiagaraMeshRendererProperties, &NiagaraSpriteRendererProperties]
  renderers [
   {rendererClass=&NiagaraMeshRendererProperties, rendererIndex=0}
   {rendererClass=&NiagaraSpriteRendererProperties, rendererIndex=1}
  ]
 }
 PopcornLow {
  emitterUpdateScript [
   &EmitterState()
   &SpawnRate(SpawnRate=4, Use Spawn Probability=false)
  ]
  particleSpawnScript [
   /SetVariables(Particles.MeshOrientation=&MakeQuatFromAxisAngle(
     Axis=&RandomRangeVector(Minimum=<-0.05 -0.1 0>, Maximum=<0.05 0.1 0>, Evaluation Type=::NewEnumerator1)
     AngleInDegrees=90
    )) {isSetParameters=true, module=SetVariables_#2}
   /SetVariables(Particles.SpawnRotationVector=&RandomRangeVector(Minimum=<0 0 0>, Maximum=<1 0.3 0>, Evaluation Type=::NewEnumerator1)) {isSetParameters=true, module=SetVariables_#3}
   &InitializeParticle(
    Lifetime Max=5
    Lifetime Min=4
    Lifetime Mode=::NewEnumerator1
    Mesh Scale Mode=::NewEnumerator1
    Position=&Add_Vector(
     A=&SimulationPosition()
     B=&MakeVector(X=0, Y=0, Z=&RandomRangeFloat(Minimum=0, Maximum=50))
    )
    Position Mode=::NewEnumerator1
    Sprite Size Mode=::NewEnumerator1
    Uniform Sprite Size Max=100
    Uniform Sprite Size Min=44
    UsePositionOffset=false
    Write Material Random=true
   ) {module=InitializeParticle001}
   &SphereLocation(
    MaskSpawnGroup=false
    Module.Override Randomness=false
    Non Uniform Scale=<1 1 0>
    Offset=<0 0 65>
    Sphere Radius=&RandomRangeFloat(Minimum=75, Maximum=140, Evaluation Type=::NewEnumerator1)
    Surface Only=true
   ) {enabled=false}
   /SetVariables(Particles.PopThreshold:NiagaraFloat=&RandomRangeFloat(Minimum=1, Maximum=2)) {isSetParameters=true, module=SetVariables_#4}
  ]
  particleUpdateScript [
   &ParticleState(Kill Particles When Lifetime Has Elapsed=false)
   /SetVariables(Particles.Spawngroup=HlslExpression() {hlslExpression="Particles.Age < Particles.PopThreshold ? 0 : 1"}) {isSetParameters=true, module=SetVariables_#5}
   &UpdateMeshOrientation(Rotation Vector=$Particles.SpawnRotationVector)
   /SetVariables(Particles.OrbitOffset:Vector3f=&RotateVectorByQuaternion(
     Vector To Rotate=&MakeVector(X=&RandomRangeFloat(Minimum=100, Maximum=150), Y=0, Z=0)
     Quaternion=&MakeQuatFromAxisAngle(
      Axis=<0 0 1>
      AngleInDegrees=&NormalizedAngleToDegrees(NormalizedAngle=&Add_Float(A=$Particles.NormalizedAge, B=&RandomRangeFloat(Minimum=0, Maximum=1)))
     )
    )) {isSetParameters=true, module=SetVariables_#6}
   /SetVariables(Particles.MeshIndex=$Particles.Spawngroup) {enabled=false, isSetParameters=true, module=SetVariables_#7}
   /SetVariables(Particles.VisibilityTag=&Add_Integer(A=$Particles.Spawngroup, B=1)) {isSetParameters=true}
   &FN_Integrator(Create Normalized Value=true, Increment=HlslExpression() {hlslExpression="Particles.Spawngroup == 1 ? 1 : 0"})
   &ScaleMeshSize(Scale Factor=&Lerp_Vector(
     A=&VectorFromFloat(Value=&FloatFromCurve(FloatCurve=DataInterface() {propertyValues {Curve {keys [<at 0 0 RCIM_Cubic>,<at 0.2 1.0034>]}}}))
     B=&VectorFromFloat(Value=&FloatFromCurve(
       FloatCurve=DataInterface() {propertyValues {Curve {keys [
           <at -0.0014 0.9856>
           {arriveTangent=-3.2738, interpMode=RCIM_Cubic, leaveTangent=-3.2738, tangentMode=RCTM_User, time=0.0796, value=2.2085}
           <at 1.0006 1.0034>
          ]}}}
       CurveIndex=$Particles.FN_Integrator.NormalizedValue
      ))
     Alpha=&MakeFloatFromInt(Int=$Particles.Spawngroup)
    ))
   /SetVariables(Particles.RenderPosition:Vector3f=&Add_Vector(A=Linked() {linkedVariable {name=Particles.Position, type=NiagaraPosition}}, B=$Particles.OrbitOffset)) {isSetParameters=true, module=SetVariables_#8}
   &AccelerationForce(Acceleration=&FN_FrontEndVector3Modifier(
     Gameplay=&MakeVector(
      X=0
      Y=0
      Z=&Lerp_Float(
       A=0
       B=&Lerp_Float(
        A=HlslExpression() {hlslExpression=Particles.Spawngroup * 2400}
        B=HlslExpression() {hlslExpression=Particles.Spawngroup * 2900}
        Alpha=$System.Dive
       )
       Alpha=&Lerp_Float(
        A=&Lerp_Float(
         A=0
         B=1
         Alpha=&ClampFloat(Float=&ScaleAndBiasFloat(Float:NiagaraFloat=$Particles.NormalizedAge, Scale:NiagaraFloat=1.1, Bias:NiagaraFloat=-0.1))
        )
        B=2
        Alpha=$System.Dive
       )
      )
     )
     Front End=&Multiply_Vector(
      A=&TransformVector(Vector=&Lerp_Vector(A=<0 0 1000>, B=<1 -1000 1000>, Alpha=$System.Dive), Destination Space=::World)
      B=&VectorFromFloat(Value=HlslExpression() {hlslExpression=Particles.Spawngroup})
     )
    ))
   &SolveForcesAndVelocity(Clamp Velocity=false, Limit Acceleration=false)
   &DynamicMaterialParameters(
    Index 0 Param 1=&FloatFromCurve(
     FloatCurve=DataInterface() {propertyValues {Curve {keys [<at 0.5016 0 RCIM_Cubic>,<at 1 1>]}}}
     CurveIndex=$Particles.FN_Integrator.NormalizedValue
    )
    Index 0 Param 4=1
    Module.Param0WriteEnabled=true
    Module.Param1WriteEnabled=true
    Module.Param2WriteEnabled=true
    Module.Param3WriteEnabled=true
   ) {module=DynamicMaterialParameters001}
   &ScaleSpriteSize(Scale Factor=&Vector2DFromFloat(Value=&FloatFromCurve(
      FloatCurve=DataInterface() {propertyValues {Curve {keys [
          {arriveTangent=4.8897, interpMode=RCIM_Cubic, leaveTangent=4.8897, tangentMode=RCTM_User, time=-0.0011, value=0.0573}
          {arriveTangent=1.3413, interpMode=RCIM_Cubic, leaveTangent=1.3413, time=0.1954, value=1.173}
          <at 1 1.4>
         ]}}}
      CurveIndex=$Particles.FN_Integrator.NormalizedValue
     )))
   /SetVariables(Particles.SmokeColor:LinearColor=&MakeLinearColorFromVectorAndFloat(
     "Vector (RGB)"=<1 1 1>
     "Float (Alpha)"=&FloatFromCurve(
      FloatCurve=DataInterface() {propertyValues {Curve {keys [
          <at -0.0008 1>
          {arriveTangent=-3.6247, interpMode=RCIM_Cubic, leaveTangent=-3.6248, tangentMode=RCTM_User, time=0.053, value=1.3351}
          <at 0.5092 0 RCIM_Cubic>
         ]}}}
      CurveIndex=$Particles.FN_Integrator.Value
      Scale Curve=0.1
     )
    )) {isSetParameters=true, module=SetVariables_#9}
   &SubUVAnimation(End Frame=64)
   &KillParticles(Kill Particles=HlslExpression() {hlslExpression="Particles.FN_Integrator.NormalizedValue > 1 ? 1 :  0"})
  ]
  rendererClasses [&NiagaraMeshRendererProperties, &NiagaraSpriteRendererProperties]
  renderers [
   {rendererClass=&NiagaraMeshRendererProperties, rendererIndex=0}
   {rendererClass=&NiagaraSpriteRendererProperties, rendererIndex=1}
   {rendererClass=&NiagaraMeshRendererProperties, rendererIndex=2}
  ]
 }
 SplashLoop {
  emitterUpdateScript [
   &EmitterState(Delay First Loop Only=true, Life Cycle Mode=::NewEnumerator1, Loop Delay=1.2, UseLoopDelay=true)
   &SpawnRate(SpawnRate=10, Use Spawn Probability=false)
  ]
  particleSpawnScript [
   /Niagara/Modules/Spawn/Initialization/InitializeParticle(
    Lifetime=&RandomRangeFloat(Minimum=0.8, Maximum=1.2, Evaluation Type=::NewEnumerator1)
    Module.Write Color=true
    Module.Write Lifetime=true
    Module.Write Mass=false
    Module.Write Material Random=true
    Module.Write Position=true
    Module.Write Scale=false
    Module.Write SpriteRotation=true
    Module.Write SpriteSize=true
    Sprite Rotation=&NormalizedAngleToDegrees(NormalizedAngle=&RandomRangeFloat(Minimum=0, Maximum=1, Evaluation Type=::NewEnumerator1))
    Sprite Size=&Vector2DFromFloat(Value=&RandomRangeFloat(Minimum=70, Maximum=90, Evaluation Type=::NewEnumerator1))
   )
   &SystemLocation(MaskSpawnGroup=false, Offset=<0 0 -25>)
   &Color(Color=&MakeLinearColorFromVectorAndFloat("Vector (RGB)"=<1 1 1>, "Float (Alpha)"=&ClampFloat(Float=0.8)))
   &DynamicMaterialParameters(Module.Param0WriteEnabled=true, Module.Param1WriteEnabled=false, Module.Param2WriteEnabled=false, Module.Param3WriteEnabled=false)
   &AddVelocity(Velocity=&RandomRangeVector(Minimum=<0 0 300>, Maximum=<0 0 400>, Evaluation Type=::NewEnumerator1))
   &TorusLocation(Align Torus With Axis=false, Large Radius=105.8244, MaskSpawnGroup=false, Module.Override Randomness=false)
  ]
  particleUpdateScript [
   &ParticleState()
   &ScaleSpriteSize(Scale Factor=&Vector2DFromFloat(Value=&FloatFromCurve(FloatCurve=DataInterface() {propertyValues {Curve {keys [<at -0.0011 0.0573 RCIM_Cubic>, {arriveTangent=1.9407, interpMode=RCIM_Cubic, leaveTangent=1.9407, time=0.4308, value=1}, <at 1 2>]}}})))
   &ScaleColor(
    Scale Alpha=&FloatFromCurve(
     FloatCurve=DataInterface() {propertyValues {Curve {keys [<at 0 0 RCIM_Cubic>,<at 0.2 1 RCIM_Cubic>,<at 1 0>]}}}
     Scale Curve=&RandomRangeFloat(Minimum=0.8, Maximum=1)
    )
    Scale RGB=&VectorFromCurve(VectorCurve=DataInterface() {propertyValues {XCurve {keys [<at 0 1>,<at 1 1>]}, YCurve {keys [<at 0 1>,<at 1 1>]}, ZCurve {keys [<at 0 1>,<at 1 1>]}}})
    ScaleA=true
    ScaleRGB=true
   )
   &DynamicMaterialParameters(
    Index 0 Param 4=1
    Module.Param0WriteEnabled=false
    Module.Param1WriteEnabled=true
    Module.Param2WriteEnabled=true
    Module.Param3WriteEnabled=true
   ) {module=DynamicMaterialParameters001}
   /Niagara/Modules/Update/SubUV/SubUVAnimation()
   &AccelerationForce(Acceleration=<0 0 -1100>)
   &SolveForcesAndVelocity(Clamp Velocity=false, Limit Acceleration=false)
  ]
  rendererClasses [&NiagaraSpriteRendererProperties]
  renderers [{rendererClass=&NiagaraSpriteRendererProperties}]
 }
 swirl1Start {
  emitterUpdateScript [
   &EmitterState()
   &SpawnBurst_Instantaneous(
    Spawn Count=12
    Spawn Probability=&MakeCustomFloatFromBool(
     Bool=&SetBoolByIntComparison(A=$Emitter.LoopCount, B=1, Comparison Type=::NewEnumerator2)
     True Float=0
     False Float=1
    )
    Use Spawn Probability=true
   )
   &SpawnRate(SpawnRate=12, Use Spawn Probability=false)
  ]
  particleSpawnScript [
   /SetVariables(Particles.MeshOrientation=&MakeQuatFromAxisAngle(Axis=<0 0 1>, AngleInDegrees=&RandomRangeFloat(Minimum=0, Maximum=720))) {isSetParameters=true}
   /SetVariables(Particles.SpawnRotationVector=&RandomRangeVector(Minimum=<0 0 -0.1>, Maximum=<0 0 -0.2>, Evaluation Type=::NewEnumerator1)) {isSetParameters=true}
   &InitializeParticle(
    Lifetime=4
    Mesh Scale=&VectorFromFloat(Value=&RandomRangeFloat(Minimum=1.5, Maximum=2.5, Evaluation Type=::NewEnumerator1))
    Module.Write Color=true
    Module.Write Lifetime=true
    Module.Write Mass=true
    Module.Write Material Random=true
    Module.Write Position=true
    Module.Write Scale=true
    Module.Write SpriteRotation=false
    Module.Write SpriteSize=false
    Position=&RandomRangeVector(Minimum=<0 0 0>, Maximum=<0 0 20>, Evaluation Type=::NewEnumerator1)
   )
   &AddVelocity(Velocity=&RandomRangeVector(Minimum=<-10 -10 5>, Maximum=<10 10 10>, Evaluation Type=::NewEnumerator1)) {enabled=false}
   /SetVariables(Particles.MeshIndex=&RandomRangeInt(Minimum=0, Maximum=1)) {isSetParameters=true}
   /SetVariables(Particles.Alpha Array=DataInterface() {propertyValues {GpuSyncMode=SyncCpuToGpu, MaxElements=0, FloatData [0.67,1]}}) {isSetParameters=true}
   /SetVariables(Particles.MaterialDynamic=DataInterface() {propertyValues {GpuSyncMode=SyncCpuToGpu, MaxElements=0, FloatData [<0 0 0 0>, <0 0 0 0>]}}) {isSetParameters=true}
   &Color(Color=&MakeLinearColorFromVectorAndFloat(
     "Vector (RGB)"=<0 0 0>
     "Float (Alpha)"=&SelectFloatFromArray(Direct Array Index=$Particles.MeshIndex, Float Selection Array=$Particles.Alpha Array)
    ))
   &DynamicMaterialParameters(
    Index 0 Param 1=&RandomRangeFloat(Minimum=0, Maximum=1, Evaluation Type=::NewEnumerator1)
    Index 0 Param 2=1
    Index 0 Param 3=1
    Index 0 Param 4=1
    Module.Param0WriteEnabled=true
    Module.Param1WriteEnabled=true
    Module.Param2WriteEnabled=true
    Module.Param3WriteEnabled=true
   )
   /SetVariables(Particles.VisibilityTag=&MakeCustomIntFromBool(
     Bool=&SetBoolByFloatComparison(A=&RandomRangeFloat(Minimum=0, Maximum=1), B=0.6)
     True INT=1
     False INT=0
    )) {isSetParameters=true}
  ]
  particleUpdateScript [
   &ParticleState()
   &SolveForcesAndVelocity(Clamp Velocity=false, Limit Acceleration=false)
   &ScaleColor(
    Scale Alpha=&Lerp_Float(
     A=&FloatFromCurve(FloatCurve=DataInterface() {propertyValues {Curve {keys [
          <at 0 0 RCIM_Cubic>
          {arriveTangent=1.0754, interpMode=RCIM_Cubic, leaveTangent=1.0754, time=0.0495, value=0.8282}
          {arriveTangent=-0.7913, interpMode=RCIM_Cubic, leaveTangent=-0.7913, tangentMode=RCTM_User, time=0.59, value=0.6345}
          <at 1.0065 -0.0178>
         ]}}})
     B=1
     Alpha=0.5
    )
    Scale RGB=&VectorFromCurve(VectorCurve=DataInterface() {propertyValues {
       XCurve {keys [<at -0.0088 1.0156>,<at 0.9896 1.0156>]}
       YCurve {keys [<at -0.0088 1.0156>,<at 0.9896 1.0156>]}
       ZCurve {keys [{arriveTangent=0.0079, time=0.0047, value=0.9925}, <at 0.9506 1.015>]}
      }})
    ScaleA=true
    ScaleRGB=true
   )
   &UpdateMeshOrientation(Rotation Rate=4.1, Rotation Vector=$Particles.SpawnRotationVector)
   /SetVariables(Particles.Color2=&MakeLinearColor(
     R:NiagaraFloat=2
     G:NiagaraFloat=2
     B:NiagaraFloat=50
     A:NiagaraFloat=&FloatFromCurve(FloatCurve=DataInterface() {propertyValues {Curve {keys [
          <at 0 0 RCIM_Cubic>
          {arriveTangent=1.0652, interpMode=RCIM_Cubic, leaveTangent=1.0652, time=0.0639, value=0.7959}
          {arriveTangent=-1.2178, interpMode=RCIM_Cubic, leaveTangent=-1.2178, tangentMode=RCTM_User, time=0.6464, value=0.6885}
          <at 1 0>
         ]}}})
    )) {isSetParameters=true}
   &ScaleMeshSize(Scale Factor=&VectorFromFloat(Value=&FloatFromCurve(
      FloatCurve=DataInterface() {propertyValues {Curve {keys [
          {arriveTangent=0.1412, interpMode=RCIM_Cubic, leaveTangent=0.1412, tangentMode=RCTM_User, time=-0.0101, value=0.5646}
          {arriveTangent=0.7799, interpMode=RCIM_Cubic, leaveTangent=0.7799, time=0.135, value=1.45}
          {arriveTangent=-1.1682, interpMode=RCIM_Cubic, leaveTangent=-1.1682, time=0.3802, value=0.869}
          {arriveTangent=0.2337, interpMode=RCIM_Cubic, leaveTangent=0.2337, time=0.5, value=1.0236}
          {arriveTangent=-0.0482, interpMode=RCIM_Cubic, leaveTangent=-0.0482, time=0.66, value=0.9344}
          <at 0.9975 0.9996 RCIM_Cubic>
         ]}}}
      CurveIndex=$Emitter.Age
     )))
  ]
  rendererClasses [&NiagaraMeshRendererProperties]
  renderers [
   {rendererClass=&NiagaraMeshRendererProperties, rendererIndex=0}
   {rendererClass=&NiagaraMeshRendererProperties, rendererIndex=1}
  ]
 }
}
summary {
 systemName=NS_Contrail_PopcornSwirl_Converted
 emitters [
  {emitterName=swirl1Start, rendererClasses [&NiagaraMeshRendererProperties]}
  {emitterName=SplashLoop, rendererClasses [&NiagaraSpriteRendererProperties]}
  {bEnabled=false, emitterName=MID_splash, rendererClasses [&NiagaraSpriteRendererProperties]}
  {
   emitterName=Popcorn
   simTarget=GPUComputeSim
   rendererClasses [&NiagaraMeshRendererProperties, &NiagaraSpriteRendererProperties]
  }
  {emitterName=PopcornLow, rendererClasses [&NiagaraMeshRendererProperties, &NiagaraSpriteRendererProperties]}
  {emitterName=InitalSplash, rendererClasses [&NiagaraSpriteRendererProperties]}
 ]
 userVariables [
  User.bIsFrontEnd {type=NiagaraBool}
  User.bIsFrontEndPreview {type=NiagaraBool}
  User.ParaGlideLeanAlpha {type=NiagaraFloat, defaultValue=0, description=""}
  User.SkeletalMesh {
   type=NiagaraDataInterfaceSkeletalMesh
   defaultValue {
    struct=NiagaraExt_VariableValue_DataInterface
    value {dataInterface=&self:NiagaraDataInterfaceSkeletalMesh_0}
   }
  }
 ]
}
